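; Parent looping gameplay cue that animates mud coverage on a target, fading it in on application and out on removal.
; Timeline TL_MudCoverage plays forward or reverses at rates derived from TimeToCover and TimeToRemove, resuming from GetLastCoverageAlpha.
; Each update passes the alpha to SetMudCoverageAlpha, stopping the timeline if it returns true.
; These two functions and GetSkeletalMesh are placeholder stubs here, while IsCoveredInMud checks the owner's mud gameplay tag.
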(event EventOnApplication(EitherVFX-preferthisone)
  (params (MyTarget :type Actor%20Object%20Reference) (Parameters :type Gameplay%20Cue%20Parameters%20Structure%20%28by%20ref%29) (ParticleComponents :type Array%20of%20FXSystem%20Component%20Object%20References) (AudioComponents :type Array%20of%20Audio%20Component%20Object%20References) (BurstCameraShakeInstance :type Camera%20Shake%20Base%20Object%20Reference) (BurstDecalInstance :type Decal%20Actor%20Object%20Reference))
  (Components|Timeline|SetPlayRate (Variables|GCNL_Mud_Covered_PrimData_Parent|GetTL_MudCoverage) (Utilities|Reciprocal (Variables|Default|GetTimeToCover)))
  (CallFunction|PlayTimeline MyTarget true))

(event EventOnRemoval(EitherVFX-preferthisversion)
  (params (MyTarget :type Actor%20Object%20Reference) (Parameters :type Gameplay%20Cue%20Parameters%20Structure%20%28by%20ref%29) (ParticleComponents :type Array%20of%20FXSystem%20Component%20Object%20References) (AudioComponents :type Array%20of%20Audio%20Component%20Object%20References) (BurstCameraShakeInstance :type Camera%20Shake%20Base%20Object%20Reference) (BurstDecalInstance :type Decal%20Actor%20Object%20Reference))
  (Components|Timeline|SetPlayRate (Variables|GCNL_Mud_Covered_PrimData_Parent|GetTL_MudCoverage) (Utilities|Reciprocal (Variables|Default|GetTimeToRemove)))
  (CallFunction|PlayTimeline MyTarget))

(event Custom|PlayTimeline
  (params (Target :type Actor%20Object%20Reference) (bForward :type Boolean))
  (Utilities|FlowControl|Sequence
    (:then_0
      (bind (_direction _coveragepercent) (Timeline|TL_MudCoverage SetNewTime :NewTime (CallFunction|GetLastCoverageAlpha Target))
        (:Update
          (bind _shouldstop (CallFunction|SetMudCoverageAlpha Target _coveragepercent bForward))
          (if _shouldstop
            (Components|Timeline|Stop (Variables|GCNL_Mud_Covered_PrimData_Parent|GetTL_MudCoverage))))))
    (:then_1
      (if bForward
        (Timeline|TL_MudCoverage Play)
        (else
          (Timeline|TL_MudCoverage Reverse))))))

(fn GetLastCoverageAlpha
  (params (TargetActor :type Actor%20Object%20Reference))
  (return 0.0))

(fn GetSkeletalMesh
  (params (Object :type Object%20Reference))
  (return 0))

(fn IsCoveredInMud
  (params (Component :type Actor%20Component%20Object%20Reference))
  (bind _returnvalue (GameplayTags|HasMatchingGameplayTag(Message) (Components|GetOwner Component) (Variables|Default|GetMudCoveredTag)))
  (return _returnvalue))

(fn SetMudCoverageAlpha
  (params (TargetActor :type Actor%20Object%20Reference) (NewAlpha :type Float%20%28double-precision%29) (Increasing :type Boolean))
  (return false))

(fn ConstructionScript)
