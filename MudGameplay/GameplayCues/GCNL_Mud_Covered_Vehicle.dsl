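; Looping gameplay cue notify for mud-covered vehicles in Fortnite, acting as an adapter to the vehicle's VehicleComponent_Mud.
; Finds the mud component by class on the target actor and caches it for later calls.
; Reads the last coverage alpha from that component and forwards new alpha and increasing values to it, returning its should-stop result.
; Also supplies the skeletal mesh by casting to FortAthenaVehicle; no update logic lives here.
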(fn GetLastCoverageAlpha
  (params (TargetActor :type Actor%20Object%20Reference))
  (bind _component (CallFunction|GetVehicleMudComponent TargetActor))
  (Utilities|IsValid _component
    (:Is%20Valid
      (return (Variables|/MudGameplay/Gameplay/Components/VehicleComponent_Mud.VehicleComponent_Mud_C|GetLastMudCoverageAlpha _component)))))

(fn GetSkeletalMesh
  (params (Object :type Object%20Reference))
  (bind _asfort_athena_vehicle (Utilities|Casting|CastToFortAthenaVehicle Object))
  (return (Variables|/Script/FortniteGame.FortAthenaVehicle|GetSkeletalMesh _asfort_athena_vehicle)))

(fn GetVehicleMudComponent
  (params (TargetActor :type Actor%20Object%20Reference))
  (bind _returnvalue (Actor|GetComponentbyClass TargetActor "/MudGameplay/Gameplay/Components/VehicleComponent_Mud.VehicleComponent_Mud_C"))
  (bind _vehiclemudcomponent (Variables|Default|GetVehicleMudComponent)
    (:then
      (return _vehiclemudcomponent))
    (:else
      (Utilities|IsValid _returnvalue
        (:Is%20Valid
          (bind _output_get (Variables|Default|SetVehicleMudComponent _returnvalue))
          (return _output_get))))))

(fn SetMudCoverageAlpha
  (params (TargetActor :type Actor%20Object%20Reference) (NewAlpha :type Float%20%28double-precision%29) (Increasing :type Boolean))
  (bind _component (CallFunction|GetVehicleMudComponent TargetActor))
  (bind _shouldstop (Class|VehicleComponentMud|SetMudCoverageAlpha _component NewAlpha Increasing))
  (return _shouldstop))

(fn ConstructionScript)
