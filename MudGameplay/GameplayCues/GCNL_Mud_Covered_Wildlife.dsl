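; Looping gameplay cue for mud coverage on wildlife pawns, supplying wildlife-specific accessors for the coverage effect.
; It resolves the pawn's PawnComponent_Mud_Wildlife, caching the reference after the first valid lookup.
; Getting the last coverage alpha and setting a new amount both go through that component, with the setter returning a should-stop result.
; The skeletal mesh comes from the Body part of the wildlife pawn's character parts component.
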(fn GetLastCoverageAlpha
  (params (TargetActor :type Actor%20Object%20Reference))
  (bind _returnvalue (CallFunction|GetWildlifeMudComponent TargetActor))
  (Utilities|IsValid _returnvalue
    (:Is%20Valid
      (return (Variables|/MudGameplay/Gameplay/Components/PawnComponent_Mud_Wildlife.PawnComponent_Mud_Wildlife_C|GetLastMudCoverageAlpha _returnvalue)))))

(fn GetSkeletalMesh
  (params (Object :type Object%20Reference))
  (bind _asnpc_pawn_wildlife_parent (Utilities|Casting|CastToNPC_Pawn_Wildlife_Parent Object))
  (bind _returnvalue (Customization|GetSkeletalMeshforPartType (Actor|GetComponentbyClass _asnpc_pawn_wildlife_parent "/Script/FortniteGame.FortCharacterPartsComponent") "Body" :!class /Script/FortniteGame.FortCharacterPartsComponent))
  (return _returnvalue))

(fn GetWildlifeMudComponent
  (params (TargetActor :type Actor%20Object%20Reference))
  (bind _returnvalue (Actor|GetComponentbyClass TargetActor "/MudGameplay/Gameplay/Components/PawnComponent_Mud_Wildlife.PawnComponent_Mud_Wildlife_C"))
  (bind _wildlifemudcomponent (Variables|Default|GetWildlifeMudComponent)
    (:then
      (return _wildlifemudcomponent))
    (:else
      (Utilities|IsValid _returnvalue
        (:Is%20Valid
          (bind _output_get (Variables|Default|SetWildlifeMudComponent _returnvalue))
          (return _output_get))))))

(fn SetMudCoverageAlpha
  (params (TargetActor :type Actor%20Object%20Reference) (NewAlpha :type Float%20%28double-precision%29) (Increasing :type Boolean))
  (bind _returnvalue (CallFunction|GetWildlifeMudComponent TargetActor))
  (bind _shouldstop (Class|PawnComponentMudWildlife|SetMudCoverageAmount _returnvalue NewAlpha Increasing))
  (return _shouldstop))

(fn ConstructionScript)
