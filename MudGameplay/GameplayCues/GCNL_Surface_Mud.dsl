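; Looping gameplay cue for mud surfaces that drives Niagara footprint and landing splash effects on a player pawn.
; On start it caches the looping footstep component, destroys it for pawns off the local team on non-mobile platforms outside replays, and binds the pawn's landed event.
; Footsteps and landings feed foot socket transforms, ground normal and velocity into Niagara, reusing the landing system once spawned.
; Removal unbinds the landed event so landing footprints stop appearing off mud.
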(event EventOnLoopingStart(Niagara)
  (params (MyTarget :type Actor%20Object%20Reference) (Parameters :type Gameplay%20Cue%20Parameters%20Structure%20%28by%20ref%29) (ParticleComponents :type Array%20of%20Niagara%20Particle%20System%20Component%20Object%20References) (AudioComponents :type Array%20of%20Audio%20Component%20Object%20References))
  (bind _returnvalue (Fortnite|GetFirstLocalFortPlayerController 0 false true))
  (bind _asfort_replay_spectator_athena (Utilities|Casting|CastToFortReplaySpectatorAthena _returnvalue))
  (bind _output (Utilities|Array|Get(acopy) ParticleComponents))
  (Utilities|FlowControl|Sequence
    (:then_0
      (Variables|Default|SetbNotValid false)
      (Variables|Default|SetbReplayMode false)
      (if (not (Fortnite|IsMobilePlatform))
        (if (Replay|GetIsPlayingReplay 0)
          (Variables|Default|SetbReplayMode true)
          (Utilities|IsValid _asfort_replay_spectator_athena
            (:Is%20Valid
              (Variables|Default|SetReplayController _asfort_replay_spectator_athena))
            (:Is%20Not%20Valid
              (Development|PrintString 0 "SSI unexcept replay issue.")))
          (elif (Team|OnLocalPlayersTeam MyTarget)
            _
            (else
              (Variables|Default|SetbNotValid true :!label _label)
              (Utilities|FlowControl|DelayUntilNextTick 0)
              (Components|Activation|Deactivate _output)
              (Components|DestroyComponent _output _output))))
        (Variables|Default|SetbNotValid false)
        (Variables|Default|SetFootStepFX _output)
        (Rendering|SetTranslucentSortPriority _output 150))
      (goto _label))
    (:then_1
      (EventDispatchers|BindEventtoOnPawnLanded (Variables|Default|GetTargetPawn) (delegate Custom|CE_LandingInMud)))))

(event Custom|CE_LandingInMud
  (params (Hit :type Hit%20Result%20Structure%20%28by%20ref%29))
  (bind _targetpawn (Variables|Default|GetTargetPawn))
  (bind _mesh (Variables|/Script/Engine.Character|GetMesh _targetpawn))
  (bind _returnvalue (Transformation|GetSocketTransform _mesh "ball_r" "RTS_World"))
  (bind _returnvalue_2 (Transformation|GetSocketTransform _mesh "ball_l" "RTS_World"))
  (bind _mudfootprintfx (Variables|Default|GetMudFootprintFX))
  (bind _targetpawn_2 (Variables|Default|GetTargetPawn))
  (Utilities|FlowControl|Delay 0 0.1)
  (bind _mudfootprintfx_2 (Variables|Default|GetMudFootprintFX)
    (:then
      (Components|Activation|Activate _mudfootprintfx_2 true))
    (:else
      (bind _returnvalue_3 (Niagara|SpawnSystemAttached "/Game/Effects/MudInteraction/NS_Mud_Splash_Landing.NS_Mud_Splash_Landing" (Variables|/Script/Engine.Character|GetMesh (Variables|Default|GetTargetPawn)) "root" "0, 0, 0" "0, 0, 0" "KeepRelativeOffset" true))
      (Variables|Default|SetMudFootprintFX _returnvalue_3)))
  (Development|PrintString 0 "LandinMud" true true "(R=0.229167,G=0.109851,B=0.051729,A=1.000000)")
  (Niagara|NiagaraSetPositionArray _mudfootprintfx "FootPositionArray" (Utilities|Array|MakeArray (.location _returnvalue_2) (.location _returnvalue)))
  (Niagara|NiagaraSetVectorArray _mudfootprintfx "FootDirectionArray" (Utilities|Array|MakeArray (Math|Vector|GetForwardVector (.rotation _returnvalue_2)) (* (Math|Vector|GetForwardVector (.rotation _returnvalue)) "-1.000000,-1.000000,-1.000000")))
  (Niagara|SetNiagaraVariable(Vector3) _mudfootprintfx "GroundNormal" (AthenaPawn|GetCachedStepSurfaceNormal _targetpawn_2))
  (Niagara|SetNiagaraVariable(Vector3) _mudfootprintfx "BoneSpeed" (Transformation|GetVelocity _targetpawn_2)))

(event EventOnRemoval(EitherVFX-preferthisversion)
  (params (MyTarget :type Actor%20Object%20Reference) (Parameters :type Gameplay%20Cue%20Parameters%20Structure%20%28by%20ref%29) (ParticleComponents :type Array%20of%20FXSystem%20Component%20Object%20References) (AudioComponents :type Array%20of%20Audio%20Component%20Object%20References) (BurstCameraShakeInstance :type Camera%20Shake%20Base%20Object%20Reference) (BurstDecalInstance :type Decal%20Actor%20Object%20Reference))
  ;[1 1 1] Unbind Landed callback to solve issue of jump land footprints appearing on other surfaces after leaving mud
  (bind _asfort_player_pawn (Utilities|Casting|CastToFortPlayerPawn MyTarget))
  (EventDispatchers|UnbindEventfromOnPawnLanded _asfort_player_pawn (EventDispatchers|CreateEvent :!event Custom|CE_LandingInMud)))

(fn OnPlayerFootstep
  (params (FootIndex :type Integer))
  (bind _targetpawn (Variables|Default|GetTargetPawn))
  (bind _returnvalue (AthenaPawn|GetCachedFootstepIndex _targetpawn))
  (bind _mesh (Variables|/Script/Engine.Character|GetMesh _targetpawn))
  (bind _returnvalue_2 (Transformation|GetSocketTransform _mesh "ball_r" "RTS_World"))
  (bind _returnvalue_3 (== _returnvalue 1))
  (bind _returnvalue_4 (Transformation|GetSocketTransform _mesh "ball_l" "RTS_World"))
  (bind _returnvalue_5 (Math|Transform|SelectTransform _returnvalue_2 _returnvalue_4 _returnvalue_3))
  (bind _targetpawn_2 (Variables|Default|GetTargetPawn))
  (bind _replaycontroller (Variables|Default|GetReplayController))
  (bind _footstepfx (Variables|Default|GetFootStepFX))
  (|CallParentFunction FootIndex)
  (if (Variables|Default|GetbNotValid)
    _
    (elif (Variables|Default|GetbReplayMode)
      (if (Utilities|IsValid _replaycontroller :!class /Script/Engine.KismetSystemLibrary)
        (if (Team|OnSameTeam (Variables|Default|GetTargetPawn) (Pawn|GetViewTarget _replaycontroller))
          _))))
  ;[1 1 1] ForTest
  (Utilities|IsValid (Variables|Default|GetFootStepFX)
    (:Is%20Valid
      (if (not (Variables|Default|GetIs%20in%20Water))
        (Niagara|NiagaraSetVector4Array _footstepfx "SpawnList" (Utilities|Array|MakeArray (Math|Vector4|MakeVector4 (.x (.location _returnvalue_5)) (.y (.location _returnvalue_5)) (.z (.location _returnvalue_5)) 1.0)))
        (Niagara|SetNiagaraVariable(Vector3) _footstepfx "FootDirection" (Math|Vector|SelectVector (* (Math|Vector|GetForwardVector (.rotation _returnvalue_5)) "-1.000000,-1.000000,-1.000000") (Math|Vector|GetForwardVector (.rotation _returnvalue_5)) _returnvalue_3))
        (Effects|Components|ParticleSystem|SetBoolParameter _footstepfx "bIsRight" _returnvalue_3)
        (Niagara|SetNiagaraVariable(Vector3) _footstepfx "GroundNormal" (AthenaPawn|GetCachedStepSurfaceNormal _targetpawn_2))
        (Niagara|SetNiagaraVariable(Vector3) _footstepfx "BoneSpeed" (Transformation|GetVelocity _targetpawn_2))))))

(fn ConstructionScript
  (|CallParentFunction))
